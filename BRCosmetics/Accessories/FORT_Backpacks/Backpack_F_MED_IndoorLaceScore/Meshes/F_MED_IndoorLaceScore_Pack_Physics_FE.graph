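# Physics asset for the IndoorLaceScore backpack cosmetic, simulating a dangling character with head, torso, pelvis, two-part arms and legs, and long four-segment ears.
# Capsule dyn_ bodies form chains with tight swing limits and little or no twist, the ears swinging widest on one axis.
# Tiny base_ anchor bodies start the torso, arm and ear chains, linked by free-angular constraints to the boxed root, torso and head respectively.
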PhysicsAsset /BRCosmetics/Accessories/FORT_Backpacks/Backpack_F_MED_IndoorLaceScore/Meshes/F_MED_IndoorLaceScore_Pack_Physics_FE
bodies {
 base_arm_1_l [{shapeType=Capsule, length=0.505, radius=0.505, shapeName=base_arm_1_l_capsule, rotation=<rot 0 -90 0>}]
 base_arm_1_r [{shapeType=Capsule, length=0.505, radius=0.505, shapeName=base_arm_1_r_capsule, rotation=<rot 0 -90 0>}]
 base_ear_01_l [{shapeType=Capsule, length=0.505, radius=0.505, shapeName=base_ear_01_l_capsule, rotation=<rot 0 -90 0>}]
 base_ear_01_r [{shapeType=Capsule, length=0.505, radius=0.505, shapeName=base_ear_01_r_capsule, rotation=<rot 0 -90 0>}]
 base_torso [{shapeType=Capsule, length=0.505, radius=0.505, shapeName=base_torso_capsule, rotation=<rot 0 -90 0>}]
 dyn_arm_1_l [
  {shapeType=Capsule, length=0.5732, radius=2.436, shapeName=dyn_arm_1_l_capsule, center=<3.6292 0.6407 -0.2794>, rotation=<rot 90 0.007 0>}
  {
   shapeType=Capsule
   length=2.64
   radius=1.1613
   shapeName=dyn_arm_1_l_capsule
   center=<1.1466 1.2218 -0.3565>
   rotation=<rot -0.1029 -179.9931 179.9999>
  }
 ]
 dyn_arm_1_r [
  {shapeType=Capsule, length=0.5732, radius=2.436, shapeName=dyn_arm_1_l_capsule, center=<3.7268 -0.5747 -0.2794>, rotation=<rot 90 0.007 0>}
  {
   shapeType=Capsule
   length=2.64
   radius=1.1613
   shapeName=dyn_arm_1_l_capsule
   center=<1.1885 -1.2178 -0.3564>
   rotation=<rot -0.1029 -179.9931 179.9999>
  }
 ]
 dyn_arm_2_l [{shapeType=Capsule, length=1.4368, radius=2.6419, shapeName=dyn_arm_1_l_capsule, center=<1.2081 0.1058 -0.3716>, rotation=<rot 90 0 0>}]
 dyn_arm_2_r [{shapeType=Capsule, length=1.4368, radius=2.6419, shapeName=dyn_arm_1_l_capsule, center=<1.2081 0.1058 -0.3716>, rotation=<rot 90 0 0>}]
 dyn_ear_01_l [{shapeType=Capsule, length=2.1537, radius=1.4938, shapeName=dyn_ear_01_l_capsule, center=<1.7882 0.8332 -0.1795>, rotation=<rot 0 -90 0>}]
 dyn_ear_01_r [{shapeType=Capsule, length=2.1537, radius=1.4938, shapeName=dyn_ear_01_l_capsule, center=<1.7868 -0.5145 -0.1795>, rotation=<rot 0 -90 0>}]
 dyn_ear_02_l [
  {shapeType=Capsule, length=3.4349, radius=1.4938, shapeName=dyn_ear_01_l_capsule, center=<1.7443 -0.0647 0.066>, rotation=<rot 0 -90 0>}
  {shapeType=Capsule, length=3.8737, radius=1.4938, shapeName=dyn_ear_01_l_capsule, center=<3.9927 0.0943 0.0943>, rotation=<rot 0 -90 0>}
  {shapeType=Capsule, length=3.8737, radius=1.4938, shapeName=dyn_ear_01_l_capsule, center=<5.5045 -0.0263 0.0943>, rotation=<rot 0 -90 0>}
 ]
 dyn_ear_02_r [
  {shapeType=Capsule, length=3.3217, radius=1.607, shapeName=dyn_ear_01_l_capsule, center=<1.9851 0.3959 0.261>, rotation=<rot -4.7848 -90 0>}
  {shapeType=Capsule, length=3.8737, radius=1.6974, shapeName=dyn_ear_01_l_capsule, center=<4.4823 0.2027 0.3765>, rotation=<rot -6.2002 -90 0>}
 ]
 dyn_ear_03_l [
  {shapeType=Capsule, length=4.4047, radius=1.4938, shapeName=dyn_ear_01_l_capsule, center=<1.7101 0.1297 0.3018>, rotation=<rot 0 -90 0>}
  {
   shapeType=Capsule
   length=3.7521
   radius=1.9865
   shapeName=dyn_ear_01_l_capsule
   center=<4.1974 0.2366 0.2182>
   rotation=<rot -5.6407 -89.9999 -0.0002>
  }
 ]
 dyn_ear_03_r [
  {
   shapeType=Capsule
   length=3.8552
   radius=1.696
   shapeName=dyn_ear_01_l_capsule
   center=<1.6901 0.1903 0.4552>
   rotation=<rot 2.6047 -90.003 -0.004>
  }
  {
   shapeType=Capsule
   length=3.7521
   radius=1.9865
   shapeName=dyn_ear_01_l_capsule
   center=<4.1974 0.2366 0.2182>
   rotation=<rot -0.3663 -90.0001 -0.0004>
  }
 ]
 dyn_ear_04_l [{shapeType=Capsule, length=1.8343, radius=2.2366, shapeName=dyn_ear_01_l_capsule, center=<2.2659 0.2566 -0.0305>, rotation=<rot 0 -90 0>}]
 dyn_ear_04_r [{shapeType=Capsule, length=1.8343, radius=2.2366, shapeName=dyn_ear_01_l_capsule, center=<2.6425 0.2566 -0.2925>, rotation=<rot 0 -90 0>}]
 dyn_head [
  {
   shapeType=Capsule
   length=1.5523
   radius=4.8827
   shapeName=dyn_head_capsule
   center=<4.5557 -0.1655 -0.6094>
   rotation=<rot 72.2282 -25.4337 63.463>
  }
  {shapeType=Capsule, length=4.2861, radius=3.7693, shapeName=dyn_head_capsule, center=<3.5315 -0.2065 -0.9528>, rotation=<rot 90 -90 0>}
 ]
 dyn_leg_01_l [
  {
   shapeType=Capsule
   length=2.708
   radius=1.724
   shapeName=dyn_leg_01_l_capsule
   center=<3.1637 -0.9837 0.1882>
   rotation=<rot 1.3756 -180 180>
  }
  {shapeType=Capsule, length=2.708, radius=1.724, shapeName=dyn_leg_01_l_capsule, center=<2.9175 1.4062 0.1941>, rotation=<rot 1.3756 -180 180>}
 ]
 dyn_leg_01_r [
  {
   shapeType=Capsule
   length=2.708
   radius=1.724
   shapeName=dyn_leg_01_l_capsule
   center=<3.263 -1.2069 0.1643>
   rotation=<rot 3.426 -179.8353 -175.4062>
  }
  {
   shapeType=Capsule
   length=2.708
   radius=1.724
   shapeName=dyn_leg_01_l_capsule
   center=<3.0169 1.183 0.1702>
   rotation=<rot 3.426 -179.8353 -175.4062>
  }
 ]
 dyn_leg_02_l [{
   shapeType=Capsule
   length=3.247
   radius=3.1051
   shapeName=dyn_leg_02_l_capsule
   center=<2.284 -0.0651 -0.5175>
   rotation=<rot -4.2432 -90 1.8556>
  }]
 dyn_leg_02_r [{
   shapeType=Capsule
   length=3.247
   radius=3.1051
   shapeName=dyn_leg_02_l_capsule
   center=<2.284 -0.0651 -0.5175>
   rotation=<rot -4.2432 -90 1.8556>
  }]
 dyn_pelvis [{shapeType=Capsule, length=3.5742, radius=4.6492, shapeName=dyn_torso_capsule, center=<1.8102 0.0292 0.6055>, rotation=<rot 0 0 -90>}]
 dyn_torso [
  {shapeType=Capsule, length=0.1306, radius=5.8055, shapeName=dyn_torso_capsule, center=<-1.1516 0.0292 0.1352>, rotation=<rot 0 -90 -90>}
  {shapeType=Capsule, length=0.584, radius=4.95, shapeName=dyn_torso_capsule, center=<2.7998 0.0292 -0.1789>, rotation=<rot 0 -90 -90>}
 ]
 root [{shapeType=Box, extentX=8.5808, extentY=63.6646, extentZ=46.5722, shapeName=root_box, center=<0 3.951 -3.5957>, rotation=<rot 0 -90 0>}]
}
constraints [
 {bone1Name=dyn_torso, bone2Name=base_torso, swing1LimitDegrees=10, swing2LimitDegrees=5, twistLimitDegrees=0}
 {bone1Name=dyn_arm_1_l, bone2Name=base_arm_1_l, swing1LimitDegrees=20, swing2LimitDegrees=15, twistLimitDegrees=0}
 {bone1Name=dyn_arm_2_l, bone2Name=dyn_arm_1_l, swing1LimitDegrees=5, swing2LimitDegrees=15, twistLimitDegrees=0}
 {bone1Name=dyn_head, bone2Name=dyn_torso, swing1LimitDegrees=10, swing2LimitDegrees=10, twistLimitDegrees=0}
 {bone1Name=dyn_ear_01_l, bone2Name=base_ear_01_l, swing1LimitDegrees=30, swing2LimitDegrees=5, twistLimitDegrees=5}
 {bone1Name=dyn_ear_02_l, bone2Name=dyn_ear_01_l, swing1LimitDegrees=30, swing2LimitDegrees=5, twistLimitDegrees=5}
 {bone1Name=dyn_ear_03_l, bone2Name=dyn_ear_02_l, swing1LimitDegrees=30, swing2LimitDegrees=7, twistLimitDegrees=10}
 {bone1Name=dyn_ear_04_l, bone2Name=dyn_ear_03_l, swing1LimitDegrees=30, swing2LimitDegrees=7, twistLimitDegrees=10}
 {bone1Name=dyn_pelvis, bone2Name=dyn_torso, swing1LimitDegrees=10, swing2LimitDegrees=5, twistLimitDegrees=0}
 {bone1Name=dyn_leg_01_l, bone2Name=dyn_pelvis, swing1LimitDegrees=15, swing2LimitDegrees=10, twistLimitDegrees=0}
 {bone1Name=dyn_leg_02_l, bone2Name=dyn_leg_01_l, swing1LimitDegrees=10, swing2LimitDegrees=10, twistLimitDegrees=0}
 {bone1Name=dyn_arm_1_r, bone2Name=base_arm_1_r, swing1LimitDegrees=20, swing2LimitDegrees=15, twistLimitDegrees=0}
 {bone1Name=dyn_ear_01_r, bone2Name=base_ear_01_r, swing1LimitDegrees=30, swing2LimitDegrees=5, twistLimitDegrees=5}
 {bone1Name=dyn_ear_02_r, bone2Name=dyn_ear_01_r, swing1LimitDegrees=30, swing2LimitDegrees=5, twistLimitDegrees=5}
 {bone1Name=dyn_ear_03_r, bone2Name=dyn_ear_02_r, swing1LimitDegrees=30, swing2LimitDegrees=7, twistLimitDegrees=10}
 {bone1Name=dyn_ear_04_r, bone2Name=dyn_ear_03_r, swing1LimitDegrees=30, swing2LimitDegrees=7, twistLimitDegrees=10}
 {bone1Name=dyn_leg_01_r, bone2Name=dyn_pelvis, swing1LimitDegrees=15, swing2LimitDegrees=10, twistLimitDegrees=0}
 {bone1Name=dyn_leg_02_r, bone2Name=dyn_leg_01_r, swing1LimitDegrees=10, swing2LimitDegrees=10, twistLimitDegrees=0}
 {bone1Name=dyn_arm_2_r, bone2Name=dyn_arm_1_r, swing1LimitDegrees=5, swing2LimitDegrees=15, twistLimitDegrees=0}
 {
  bone1Name=base_torso
  bone2Name=root
  swing1LimitDegrees=45
  swing1Motion=Free
  swing2LimitDegrees=45
  swing2Motion=Free
  twistLimitDegrees=45
  twistMotion=Free
 }
 {
  bone1Name=base_arm_1_r
  bone2Name=dyn_torso
  swing1LimitDegrees=45
  swing1Motion=Free
  swing2LimitDegrees=45
  swing2Motion=Free
  twistLimitDegrees=45
  twistMotion=Free
 }
 {
  bone1Name=base_arm_1_l
  bone2Name=dyn_torso
  swing1LimitDegrees=45
  swing1Motion=Free
  swing2LimitDegrees=45
  swing2Motion=Free
  twistLimitDegrees=45
  twistMotion=Free
 }
 {
  bone1Name=base_ear_01_l
  bone2Name=dyn_head
  swing1LimitDegrees=45
  swing1Motion=Free
  swing2LimitDegrees=45
  swing2Motion=Free
  twistLimitDegrees=45
  twistMotion=Free
 }
 {
  bone1Name=base_ear_01_r
  bone2Name=dyn_head
  swing1LimitDegrees=45
  swing1Motion=Free
  swing2LimitDegrees=45
  swing2Motion=Free
  twistLimitDegrees=45
  twistMotion=Free
 }
]
