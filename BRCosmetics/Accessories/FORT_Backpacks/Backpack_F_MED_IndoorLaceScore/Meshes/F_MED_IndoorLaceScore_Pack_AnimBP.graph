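# Animation Blueprint for the IndoorLaceScore backpack cosmetic, adding secondary motion to its pelvis, torso, head, limb and ear bones.
# The root-adjusted input pose is cached and feeds three bool-blended branches: RigidBody alone, AnimDynamics body chain plus front-end RigidBody, or full AnimDynamics chain including ears.
# The event graph computes gravity override from a parameter asset, switches front-end/BR sim-space settings and ear offsets, and records wearer gender.
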AnimBlueprint /BRCosmetics/Accessories/FORT_Backpacks/Backpack_F_MED_IndoorLaceScore/Meshes/F_MED_IndoorLaceScore_Pack_AnimBP
target_skeleton=&/F_MED_IndoorLaceScore_Pack_Skeleton
graphs [
 AnimGraph {nodes [
   Output Pose(Result=Use cached pose 'DynamicsCache'()) {layerGroup=None}
   n1 = Input Pose()
   n2 = "Transform (Modify) Bone - Bone: root"!(
    ComponentPose="Transform (Modify) Bone - Bone: root"!(ComponentPose=Local To Component(LocalPose=@n1), Alpha:float, Translation="0,2.5,0") -> (Pose) {translationMode=BMM_Additive, translationSpace=BCS_BoneSpace, boneToModify {boneName=root}, translation {y=2.5}}
    bAlphaBoolEnabled:bool=false
    Translation="0,-1,0"
   ) -> (Pose) {alphaInputType=Bool, translationMode=BMM_Additive, translationSpace=BCS_BoneSpace, boneToModify {boneName=root}, translation {y=-1}}
   Save cached pose 'CRCache_Input'(Pose=Component To Local(ComponentPose=@n2)) {cacheName=CRCache_Input}
   n3 = RigidBody!(
    ComponentPose=Local To Component(LocalPose=Use cached pose 'CRCache_Input'())
    Alpha=0.95
    OverrideWorldGravity:Vector="0,0,-980"
    SimSpaceSettings:Sim Space Settings Structure=(WorldAlpha=1,VelocityScaleZ=0.5,DampingAlpha=0.25,MaxLinearVelocity=300,MaxAngularVelocity=4,MaxLinearAcceleration=1000,MaxAngularAcceleration=1600,ExternalLinearDragV=(X=1.2,Y=1.2,Z=1.2),ExternalLinearVelocity=(X=0,Y=0,Z=0),ExternalAngularVelocity=(X=0,Y=0,Z=0))
   ) -> (Pose) {
    bOverrideWorldGravity=true
    overridePhysicsAsset=&/F_MED_IndoorLaceScore_Pack_Physics
    simulationSpace=BaseBoneSpace
    alphaBoolBlend {blendOption=Linear}
    baseBoneRef {boneName=pack}
    overrideWorldGravity {z=-980}
    simSpaceSettings {
     dampingAlpha=0.25
     maxAngularAcceleration=1600
     maxAngularVelocity=4
     maxLinearAcceleration=1000
     maxLinearVelocity=300
     velocityScaleZ=0.5
     worldAlpha=1
     externalLinearDragV=<1.2 1.2 1.2>
    }
   }
   n4 = Component To Local(ComponentPose=@n3)
   n5 = Anim Dynamics - Bone: dyn_pelvis!(ComponentPose=Local To Component(LocalPose=Use cached pose 'CRCache_Input'()), Alpha=0.5, GravityOverride:Vector="0,0,-980") -> (Pose) {
    angularDampingOverride=1
    angularSpringConstant=10000
    bAngularSpring=true
    bOverrideAngularDamping=true
    bUseGravityOverride=true
    gravityScale=0.35
    simulationSpace=BoneRelative
    boundBone {boneName=dyn_pelvis}
    componentAppliedLinearAccClamp=<10000 10000 10000>
    componentLinearAccScale=<1 1 1>
    componentLinearVelScale=<1 1 1>
    gravityOverride {z=-980}
    physicsBodyDefinitions [{
      boundBone {boneName=dyn_pelvis}
      boxExtents=<20 30 30>
      constraintSetup {angularLimitsMax=<5 5 15>, angularLimitsMin=<-5 -5 -15>, angularTarget=<1 0 0>}
      localJointOffset=<1 0 0>
     }]
    relativeSpaceBone {boneName=pack}
    simSpaceSettings {maxAngularAcceleration=3, maxAngularVelocity=2, simSpaceAngularAlpha=1}
   }
   n6 = Anim Dynamics - Bone: dyn_torso!(ComponentPose=@n5, Alpha=0.5, GravityOverride:Vector="0,0,-980") -> (Pose) {
    angularDampingOverride=1
    angularSpringConstant=10000
    bAngularSpring=true
    bOverrideAngularDamping=true
    bUseGravityOverride=true
    gravityScale=0.35
    simulationSpace=BoneRelative
    boundBone {boneName=dyn_torso}
    componentAppliedLinearAccClamp=<10000 10000 10000>
    componentLinearAccScale=<1 1 1>
    componentLinearVelScale=<1 1 1>
    gravityOverride {z=-980}
    physicsBodyDefinitions [{
      boundBone {boneName=dyn_torso}
      boxExtents=<20 30 30>
      constraintSetup {angularLimitsMax=<5 5 15>, angularLimitsMin=<-5 -5 -15>, angularTarget=<1 0 0>}
      localJointOffset=<1 0 0>
     }]
    relativeSpaceBone {boneName=pack}
    simSpaceSettings {maxAngularAcceleration=3, maxAngularVelocity=2, simSpaceAngularAlpha=1}
   }
   n7 = Anim Dynamics - Bone: dyn_head!(ComponentPose=@n6, Alpha=0.5, GravityOverride:Vector="0,0,-980") -> (Pose) {
    angularDampingOverride=1
    angularSpringConstant=10000
    bAngularSpring=true
    bOverrideAngularDamping=true
    bUseGravityOverride=true
    gravityScale=0.35
    simulationSpace=BoneRelative
    boundBone {boneName=dyn_head}
    componentAppliedLinearAccClamp=<10000 10000 10000>
    componentLinearAccScale=<1 1 1>
    componentLinearVelScale=<1 1 1>
    gravityOverride {z=-980}
    physicsBodyDefinitions [{
      boundBone {boneName=dyn_head}
      boxExtents=<20 30 30>
      constraintSetup {angularLimitsMax=<5 5 15>, angularLimitsMin=<-5 -5 -15>, angularTarget=<1 0 0>}
      localJointOffset=<1 0 0>
     }]
    relativeSpaceBone {boneName=pack}
    simSpaceSettings {maxAngularAcceleration=3, maxAngularVelocity=2, simSpaceAngularAlpha=1}
   }
   n8 = Anim Dynamics - Bone: dyn_leg_01_l!(ComponentPose=@n7, Alpha=0.75, GravityOverride:Vector="0,0,-980") -> (Pose) {
    angularDampingOverride=1
    angularSpringConstant=15000
    bAngularSpring=true
    bOverrideAngularDamping=true
    bUseGravityOverride=true
    gravityScale=0.35
    simulationSpace=BoneRelative
    boundBone {boneName=dyn_leg_01_l}
    componentAppliedLinearAccClamp=<10000 10000 10000>
    componentLinearAccScale=<1 1 1>
    componentLinearVelScale=<1 1 1>
    gravityOverride {z=-980}
    physicsBodyDefinitions [{
      boundBone {boneName=dyn_leg_01_l}
      boxExtents=<25 35 35>
      constraintSetup {angularLimitsMax=<15 30 15>, angularLimitsMin=<-15 -15 -15>, angularTarget=<1 0 0>}
      localJointOffset=<1.5 0 0>
     }]
    relativeSpaceBone {boneName=pack}
    simSpaceSettings {maxAngularAcceleration=3, maxAngularVelocity=2, simSpaceAngularAlpha=1}
   }
   n9 = Anim Dynamics - Bone: dyn_leg_02_l!(ComponentPose=@n8, Alpha=0.75, GravityOverride:Vector="0,0,-980") -> (Pose) {
    angularDampingOverride=1
    angularSpringConstant=15000
    bAngularSpring=true
    bOverrideAngularDamping=true
    bUseGravityOverride=true
    gravityScale=0.35
    simulationSpace=BoneRelative
    boundBone {boneName=dyn_leg_02_l}
    componentAppliedLinearAccClamp=<10000 10000 10000>
    componentLinearAccScale=<1 1 1>
    componentLinearVelScale=<1 1 1>
    gravityOverride {z=-980}
    physicsBodyDefinitions [{
      boundBone {boneName=dyn_leg_02_l}
      boxExtents=<25 35 35>
      constraintSetup {angularLimitsMax=<15 30 15>, angularLimitsMin=<-15 -15 -15>, angularTarget=<1 0 0>}
      localJointOffset=<1.5 0 0>
     }]
    relativeSpaceBone {boneName=pack}
    simSpaceSettings {maxAngularAcceleration=3, maxAngularVelocity=2, simSpaceAngularAlpha=1}
   }
   n10 = Anim Dynamics - Bone: dyn_leg_01_r!(ComponentPose=@n9, Alpha=0.75, GravityOverride:Vector="0,0,-980") -> (Pose) {
    angularDampingOverride=1
    angularSpringConstant=15000
    bAngularSpring=true
    bOverrideAngularDamping=true
    bUseGravityOverride=true
    gravityScale=0.35
    simulationSpace=BoneRelative
    boundBone {boneName=dyn_leg_01_r}
    componentAppliedLinearAccClamp=<10000 10000 10000>
    componentLinearAccScale=<1 1 1>
    componentLinearVelScale=<1 1 1>
    gravityOverride {z=-980}
    physicsBodyDefinitions [{
      boundBone {boneName=dyn_leg_01_r}
      boxExtents=<25 35 35>
      constraintSetup {angularLimitsMax=<15 30 15>, angularLimitsMin=<-15 -15 -15>, angularTarget=<1 0 0>}
      localJointOffset=<1.5 0 0>
     }]
    relativeSpaceBone {boneName=pack}
    simSpaceSettings {maxAngularAcceleration=3, maxAngularVelocity=2, simSpaceAngularAlpha=1}
   }
   n11 = Anim Dynamics - Bone: dyn_leg_02_r!(ComponentPose=@n10, Alpha=0.75, GravityOverride:Vector="0,0,-980") -> (Pose) {
    angularDampingOverride=1
    angularSpringConstant=15000
    bAngularSpring=true
    bOverrideAngularDamping=true
    bUseGravityOverride=true
    gravityScale=0.35
    simulationSpace=BoneRelative
    boundBone {boneName=dyn_leg_02_r}
    componentAppliedLinearAccClamp=<10000 10000 10000>
    componentLinearAccScale=<1 1 1>
    componentLinearVelScale=<1 1 1>
    gravityOverride {z=-980}
    physicsBodyDefinitions [{
      boundBone {boneName=dyn_leg_02_r}
      boxExtents=<25 35 35>
      constraintSetup {angularLimitsMax=<15 30 15>, angularLimitsMin=<-15 -15 -15>, angularTarget=<1 0 0>}
      localJointOffset=<1.5 0 0>
     }]
    relativeSpaceBone {boneName=pack}
    simSpaceSettings {maxAngularAcceleration=3, maxAngularVelocity=2, simSpaceAngularAlpha=1}
   }
   n12 = Anim Dynamics - Bone: dyn_arm_1_l!(ComponentPose=@n11, Alpha=0.75, GravityOverride:Vector="0,0,-980") -> (Pose) {
    angularDampingOverride=1
    angularSpringConstant=15000
    bAngularSpring=true
    bOverrideAngularDamping=true
    bUseGravityOverride=true
    gravityScale=0.35
    simulationSpace=BoneRelative
    boundBone {boneName=dyn_arm_1_l}
    componentAppliedLinearAccClamp=<10000 10000 10000>
    componentLinearAccScale=<1 1 1>
    componentLinearVelScale=<1 1 1>
    gravityOverride {z=-980}
    physicsBodyDefinitions [{
      boundBone {boneName=dyn_arm_1_l}
      boxExtents=<25 35 35>
      constraintSetup {angularLimitsMax=<15 15 30>, angularLimitsMin=<-15 -5 -15>, angularTarget=<1 0 0>}
      localJointOffset=<2 0 0>
     }]
    relativeSpaceBone {boneName=pack}
    simSpaceSettings {maxAngularAcceleration=3, maxAngularVelocity=2, simSpaceAngularAlpha=1}
   }
   n13 = Anim Dynamics - Bone: dyn_arm_2_l!(ComponentPose=@n12, Alpha=0.75, GravityOverride:Vector="0,0,-980") -> (Pose) {
    angularDampingOverride=1
    angularSpringConstant=15000
    bAngularSpring=true
    bOverrideAngularDamping=true
    bUseGravityOverride=true
    gravityScale=0.35
    simulationSpace=BoneRelative
    boundBone {boneName=dyn_arm_2_l}
    componentAppliedLinearAccClamp=<10000 10000 10000>
    componentLinearAccScale=<1 1 1>
    componentLinearVelScale=<1 1 1>
    gravityOverride {z=-980}
    physicsBodyDefinitions [{
      boundBone {boneName=dyn_arm_2_l}
      boxExtents=<25 35 35>
      constraintSetup {angularLimitsMax=<15 30 30>, angularLimitsMin=<-15 -5 -15>, angularTarget=<1 0 0>}
      localJointOffset=<2 0 0>
     }]
    relativeSpaceBone {boneName=pack}
    simSpaceSettings {maxAngularAcceleration=3, maxAngularVelocity=2, simSpaceAngularAlpha=1}
   }
   n14 = Anim Dynamics - Bone: dyn_arm_1_r!(ComponentPose=@n13, Alpha=0.75, GravityOverride:Vector="0,0,-980") -> (Pose) {
    angularDampingOverride=1
    angularSpringConstant=15000
    bAngularSpring=true
    bOverrideAngularDamping=true
    bUseGravityOverride=true
    gravityScale=0.35
    simulationSpace=BoneRelative
    boundBone {boneName=dyn_arm_1_r}
    componentAppliedLinearAccClamp=<10000 10000 10000>
    componentLinearAccScale=<1 1 1>
    componentLinearVelScale=<1 1 1>
    gravityOverride {z=-980}
    physicsBodyDefinitions [{
      boundBone {boneName=dyn_arm_1_r}
      boxExtents=<25 35 35>
      constraintSetup {angularLimitsMax=<15 15 15>, angularLimitsMin=<-15 -5 -30>, angularTarget=<1 0 0>}
      localJointOffset=<2 0 0>
     }]
    relativeSpaceBone {boneName=pack}
    simSpaceSettings {maxAngularAcceleration=3, maxAngularVelocity=2, simSpaceAngularAlpha=1}
   }
   n15 = Anim Dynamics - Bone: dyn_arm_2_r!(ComponentPose=@n14, Alpha=0.75, GravityOverride:Vector="0,0,-980") -> (Pose) {
    angularDampingOverride=1
    angularSpringConstant=15000
    bAngularSpring=true
    bOverrideAngularDamping=true
    bUseGravityOverride=true
    gravityScale=0.35
    simulationSpace=BoneRelative
    boundBone {boneName=dyn_arm_2_r}
    componentAppliedLinearAccClamp=<10000 10000 10000>
    componentLinearAccScale=<1 1 1>
    componentLinearVelScale=<1 1 1>
    gravityOverride {z=-980}
    physicsBodyDefinitions [{
      boundBone {boneName=dyn_arm_2_r}
      boxExtents=<25 35 35>
      constraintSetup {angularLimitsMax=<15 30 15>, angularLimitsMin=<-15 -5 -30>, angularTarget=<1 0 0>}
      localJointOffset=<2 0 0>
     }]
    relativeSpaceBone {boneName=pack}
    simSpaceSettings {maxAngularAcceleration=3, maxAngularVelocity=2, simSpaceAngularAlpha=1}
   }
   n16 = RigidBody(
    ComponentPose=@n15
    Alpha=0.95
    OverrideWorldGravity="0,0,-980"
    SimSpaceSettings=(WorldAlpha=1,VelocityScaleZ=0.5,DampingAlpha=0.25,MaxLinearVelocity=300,MaxAngularVelocity=4,MaxLinearAcceleration=1000,MaxAngularAcceleration=1600,ExternalLinearDragV=(X=1.2,Y=1.2,Z=1.2),ExternalLinearVelocity=(X=0,Y=0,Z=0),ExternalAngularVelocity=(X=0,Y=0,Z=0))
   ) {
    bOverrideWorldGravity=true
    overridePhysicsAsset=&/F_MED_IndoorLaceScore_Pack_Physics_FE
    simulationSpace=BaseBoneSpace
    alphaBoolBlend {blendOption=Linear}
    baseBoneRef {boneName=pack}
    overrideWorldGravity {z=-980}
    simSpaceSettings {
     dampingAlpha=0.25
     maxAngularAcceleration=1600
     maxAngularVelocity=4
     maxLinearAcceleration=1000
     maxLinearVelocity=300
     velocityScaleZ=0.5
     worldAlpha=1
     externalLinearDragV=<1.2 1.2 1.2>
    }
   }
   n17 = Component To Local(ComponentPose=@n16)
   n18 = Local To Component(LocalPose=Use cached pose 'CRCache_Input'())
   n19 = Anim Dynamics - Bone: dyn_pelvis(
    ComponentPose="Transform (Modify) Bone - Bone: base_ear_01_r"!(
     ComponentPose="Transform (Modify) Bone - Bone: base_ear_01_l"!(ComponentPose=@n18, Alpha=1, Rotation="0,10,0") -> (Pose) {rotationMode=BMM_Additive, rotationSpace=BCS_BoneSpace, boneToModify {boneName=base_ear_01_l}, rotation {yaw=10}}
     Alpha=1
     Rotation="0,-10,0"
    ) -> (Pose) {rotationMode=BMM_Additive, rotationSpace=BCS_BoneSpace, boneToModify {boneName=base_ear_01_r}, rotation {yaw=-10}}
    Alpha=1
    GravityOverride="0,0,-980"
   ) {
    angularDampingOverride=1
    angularSpringConstant=10000
    bAngularSpring=true
    bOverrideAngularDamping=true
    bUseGravityOverride=true
    gravityScale=0.35
    simulationSpace=RootRelative
    boundBone {boneName=dyn_pelvis}
    componentAppliedLinearAccClamp=<10000 10000 10000>
    componentLinearAccScale=<1 1 1>
    componentLinearVelScale=<1 1 1>
    gravityOverride {z=-980}
    physicsBodyDefinitions [{
      boundBone {boneName=dyn_pelvis}
      boxExtents=<20 30 30>
      constraintSetup {angularLimitsMax=<5 5 15>, angularLimitsMin=<-5 -5 -15>, angularTarget=<1 0 0>}
      localJointOffset=<2 0 0>
     }]
    simSpaceSettings {maxAngularAcceleration=3, maxAngularVelocity=1, simSpaceAngularAlpha=1}
   }
   n20 = Anim Dynamics - Bone: dyn_torso!(ComponentPose=@n19, bAlphaBoolEnabled:bool=NOT Boolean(A=Get bIsInFrontEnd()), GravityOverride="0,0,-980") -> (Pose) {
    alphaInputType=Bool
    angularDampingOverride=1
    angularSpringConstant=10000
    bAngularSpring=true
    bOverrideAngularDamping=true
    bUseGravityOverride=true
    gravityScale=0.35
    simulationSpace=RootRelative
    boundBone {boneName=dyn_torso}
    componentAppliedLinearAccClamp=<10000 10000 10000>
    componentLinearAccScale=<1 1 1>
    componentLinearVelScale=<1 1 1>
    gravityOverride {z=-980}
    physicsBodyDefinitions [{
      boundBone {boneName=dyn_torso}
      boxExtents=<20 30 30>
      constraintSetup {angularLimitsMax=<5 5 15>, angularLimitsMin=<-5 -5 -15>, angularTarget=<1 0 0>}
      localJointOffset=<2 0 0>
     }]
    simSpaceSettings {maxAngularAcceleration=3, maxAngularVelocity=1, simSpaceAngularAlpha=1}
   }
   n21 = Anim Dynamics - Bone: dyn_head(ComponentPose=@n20, Alpha=1, GravityOverride="0,0,-980") {
    angularDampingOverride=1
    angularSpringConstant=10000
    bAngularSpring=true
    bOverrideAngularDamping=true
    bUseGravityOverride=true
    gravityScale=0.35
    simulationSpace=RootRelative
    boundBone {boneName=dyn_head}
    componentAppliedLinearAccClamp=<10000 10000 10000>
    componentLinearAccScale=<1 1 1>
    componentLinearVelScale=<1 1 1>
    gravityOverride {z=-980}
    physicsBodyDefinitions [{
      boundBone {boneName=dyn_head}
      boxExtents=<20 30 30>
      constraintSetup {angularLimitsMax=<5 5 15>, angularLimitsMin=<-5 -5 -15>, angularTarget=<1 0 0>}
      localJointOffset=<2 0 0>
     }]
    simSpaceSettings {maxAngularAcceleration=3, maxAngularVelocity=1, simSpaceAngularAlpha=1}
   }
   n22 = Anim Dynamics - Bone: dyn_leg_01_l(ComponentPose=@n21, Alpha=1, GravityOverride="0,0,-980") {
    angularDampingOverride=1
    angularSpringConstant=15000
    bAngularSpring=true
    bOverrideAngularDamping=true
    bUseGravityOverride=true
    gravityScale=0.35
    simulationSpace=RootRelative
    boundBone {boneName=dyn_leg_01_l}
    componentAppliedLinearAccClamp=<10000 10000 10000>
    componentLinearAccScale=<1 1 1>
    componentLinearVelScale=<1 1 1>
    gravityOverride {z=-980}
    physicsBodyDefinitions [{
      boundBone {boneName=dyn_leg_01_l}
      boxExtents=<25 35 35>
      constraintSetup {angularLimitsMax=<15 30 15>, angularLimitsMin=<-15 -15 -15>, angularTarget=<1 0 0>}
      localJointOffset=<3 0 0>
     }]
    simSpaceSettings {maxAngularAcceleration=3, maxAngularVelocity=1, simSpaceAngularAlpha=1}
   }
   n23 = Anim Dynamics - Bone: dyn_leg_02_l(ComponentPose=@n22, Alpha=1, GravityOverride="0,0,-980") {
    angularDampingOverride=1
    angularSpringConstant=15000
    bAngularSpring=true
    bOverrideAngularDamping=true
    bUseGravityOverride=true
    gravityScale=0.35
    simulationSpace=RootRelative
    boundBone {boneName=dyn_leg_02_l}
    componentAppliedLinearAccClamp=<10000 10000 10000>
    componentLinearAccScale=<1 1 1>
    componentLinearVelScale=<1 1 1>
    gravityOverride {z=-980}
    physicsBodyDefinitions [{
      boundBone {boneName=dyn_leg_02_l}
      boxExtents=<25 35 35>
      constraintSetup {angularLimitsMax=<15 30 15>, angularLimitsMin=<-15 -15 -15>, angularTarget=<1 0 0>}
      localJointOffset=<3 0 0>
     }]
    simSpaceSettings {maxAngularAcceleration=3, maxAngularVelocity=1, simSpaceAngularAlpha=1}
   }
   n24 = Anim Dynamics - Bone: dyn_leg_01_r(ComponentPose=@n23, Alpha=1, GravityOverride="0,0,-980") {
    angularDampingOverride=1
    angularSpringConstant=15000
    bAngularSpring=true
    bOverrideAngularDamping=true
    bUseGravityOverride=true
    gravityScale=0.35
    simulationSpace=RootRelative
    boundBone {boneName=dyn_leg_01_r}
    componentAppliedLinearAccClamp=<10000 10000 10000>
    componentLinearAccScale=<1 1 1>
    componentLinearVelScale=<1 1 1>
    gravityOverride {z=-980}
    physicsBodyDefinitions [{
      boundBone {boneName=dyn_leg_01_r}
      boxExtents=<25 35 35>
      constraintSetup {angularLimitsMax=<15 30 15>, angularLimitsMin=<-15 -15 -15>, angularTarget=<1 0 0>}
      localJointOffset=<3 0 0>
     }]
    simSpaceSettings {maxAngularAcceleration=3, maxAngularVelocity=1, simSpaceAngularAlpha=1}
   }
   n25 = Anim Dynamics - Bone: dyn_leg_02_r(ComponentPose=@n24, Alpha=1, GravityOverride="0,0,-980") {
    angularDampingOverride=1
    angularSpringConstant=15000
    bAngularSpring=true
    bOverrideAngularDamping=true
    bUseGravityOverride=true
    gravityScale=0.35
    simulationSpace=RootRelative
    boundBone {boneName=dyn_leg_02_r}
    componentAppliedLinearAccClamp=<10000 10000 10000>
    componentLinearAccScale=<1 1 1>
    componentLinearVelScale=<1 1 1>
    gravityOverride {z=-980}
    physicsBodyDefinitions [{
      boundBone {boneName=dyn_leg_02_r}
      boxExtents=<25 35 35>
      constraintSetup {angularLimitsMax=<15 30 15>, angularLimitsMin=<-15 -15 -15>, angularTarget=<1 0 0>}
      localJointOffset=<3 0 0>
     }]
    simSpaceSettings {maxAngularAcceleration=3, maxAngularVelocity=1, simSpaceAngularAlpha=1}
   }
   n26 = Anim Dynamics - Bone: dyn_arm_1_l(ComponentPose=@n25, Alpha=1, GravityOverride="0,0,-980") {
    angularDampingOverride=1
    angularSpringConstant=15000
    bAngularSpring=true
    bOverrideAngularDamping=true
    bUseGravityOverride=true
    gravityScale=0.35
    simulationSpace=RootRelative
    boundBone {boneName=dyn_arm_1_l}
    componentAppliedLinearAccClamp=<10000 10000 10000>
    componentLinearAccScale=<1 1 1>
    componentLinearVelScale=<1 1 1>
    gravityOverride {z=-980}
    physicsBodyDefinitions [{
      boundBone {boneName=dyn_arm_1_l}
      boxExtents=<25 35 35>
      constraintSetup {angularLimitsMax=<15 15 30>, angularLimitsMin=<-15 -5 -15>, angularTarget=<1 0 0>}
      localJointOffset=<3 0 0>
     }]
    simSpaceSettings {maxAngularAcceleration=3, maxAngularVelocity=1, simSpaceAngularAlpha=1}
   }
   n27 = Anim Dynamics - Bone: dyn_arm_2_l(ComponentPose=@n26, Alpha=1, GravityOverride="0,0,-980") {
    angularDampingOverride=1
    angularSpringConstant=15000
    bAngularSpring=true
    bOverrideAngularDamping=true
    bUseGravityOverride=true
    gravityScale=0.35
    simulationSpace=RootRelative
    boundBone {boneName=dyn_arm_2_l}
    componentAppliedLinearAccClamp=<10000 10000 10000>
    componentLinearAccScale=<1 1 1>
    componentLinearVelScale=<1 1 1>
    gravityOverride {z=-980}
    physicsBodyDefinitions [{
      boundBone {boneName=dyn_arm_2_l}
      boxExtents=<25 35 35>
      constraintSetup {angularLimitsMax=<15 30 30>, angularLimitsMin=<-15 -5 -15>, angularTarget=<1 0 0>}
      localJointOffset=<3 0 0>
     }]
    simSpaceSettings {maxAngularAcceleration=3, maxAngularVelocity=1, simSpaceAngularAlpha=1}
   }
   n28 = Anim Dynamics - Bone: dyn_arm_1_r(ComponentPose=@n27, Alpha=1, GravityOverride="0,0,-980") {
    angularDampingOverride=1
    angularSpringConstant=15000
    bAngularSpring=true
    bOverrideAngularDamping=true
    bUseGravityOverride=true
    gravityScale=0.35
    simulationSpace=RootRelative
    boundBone {boneName=dyn_arm_1_r}
    componentAppliedLinearAccClamp=<10000 10000 10000>
    componentLinearAccScale=<1 1 1>
    componentLinearVelScale=<1 1 1>
    gravityOverride {z=-980}
    physicsBodyDefinitions [{
      boundBone {boneName=dyn_arm_1_r}
      boxExtents=<25 35 35>
      constraintSetup {angularLimitsMax=<15 15 15>, angularLimitsMin=<-15 -5 -30>, angularTarget=<1 0 0>}
      localJointOffset=<3 0 0>
     }]
    simSpaceSettings {maxAngularAcceleration=3, maxAngularVelocity=1, simSpaceAngularAlpha=1}
   }
   n29 = Anim Dynamics - Bone: dyn_arm_2_r(ComponentPose=@n28, Alpha=1, GravityOverride="0,0,-980") {
    angularDampingOverride=1
    angularSpringConstant=15000
    bAngularSpring=true
    bOverrideAngularDamping=true
    bUseGravityOverride=true
    gravityScale=0.35
    simulationSpace=RootRelative
    boundBone {boneName=dyn_arm_2_r}
    componentAppliedLinearAccClamp=<10000 10000 10000>
    componentLinearAccScale=<1 1 1>
    componentLinearVelScale=<1 1 1>
    gravityOverride {z=-980}
    physicsBodyDefinitions [{
      boundBone {boneName=dyn_arm_2_r}
      boxExtents=<25 35 35>
      constraintSetup {angularLimitsMax=<15 30 15>, angularLimitsMin=<-15 -5 -30>, angularTarget=<1 0 0>}
      localJointOffset=<3 0 0>
     }]
    simSpaceSettings {maxAngularAcceleration=3, maxAngularVelocity=1, simSpaceAngularAlpha=1}
   }
   n30 = Anim Dynamics - Bone: dyn_ear_01_l!(ComponentPose=@n29, Alpha=1, GravityOverride:Vector="0,0,-980") -> (Pose) {
    angularDampingOverride=1
    angularSpringConstant=15000
    bAngularSpring=true
    bOverrideAngularDamping=true
    bUseGravityOverride=true
    gravityScale=0.35
    simulationSpace=RootRelative
    boundBone {boneName=dyn_ear_01_l}
    componentAppliedLinearAccClamp=<10000 10000 10000>
    componentLinearAccScale=<1 1 1>
    componentLinearVelScale=<1 1 1>
    gravityOverride {z=-980}
    physicsBodyDefinitions [{
      boundBone {boneName=dyn_ear_01_l}
      boxExtents=<25 35 35>
      constraintSetup {angularLimitsMax=<0 0 15>, angularLimitsMin=<0 0 -10>, angularTarget=<1 0 0>}
      localJointOffset=<3.5 0 0>
     }]
    simSpaceSettings {maxAngularAcceleration=3, maxAngularVelocity=1, simSpaceAngularAlpha=1}
   }
   n31 = Anim Dynamics - Bone: dyn_ear_02_l!(ComponentPose=@n30, Alpha=1, GravityOverride:Vector="0,0,-980") -> (Pose) {
    angularDampingOverride=1
    angularSpringConstant=15000
    bAngularSpring=true
    bOverrideAngularDamping=true
    bUseGravityOverride=true
    gravityScale=0.35
    simulationSpace=RootRelative
    boundBone {boneName=dyn_ear_02_l}
    componentAppliedLinearAccClamp=<10000 10000 10000>
    componentLinearAccScale=<1 1 1>
    componentLinearVelScale=<1 1 1>
    gravityOverride {z=-980}
    physicsBodyDefinitions [{
      boundBone {boneName=dyn_ear_02_l}
      boxExtents=<25 35 35>
      constraintSetup {angularLimitsMax=<0 0 30>, angularLimitsMin=<0 0 -7.5>, angularTarget=<1 0 0>}
      localJointOffset=<4 0 0>
     }]
    simSpaceSettings {maxAngularAcceleration=3, maxAngularVelocity=1, simSpaceAngularAlpha=1}
   }
   n32 = Anim Dynamics - Bone: dyn_ear_03_l!(ComponentPose=@n31, Alpha=1, GravityOverride:Vector="0,0,-980") -> (Pose) {
    angularDampingOverride=1
    angularSpringConstant=15000
    bAngularSpring=true
    bOverrideAngularDamping=true
    bUseGravityOverride=true
    gravityScale=0.35
    simulationSpace=RootRelative
    boundBone {boneName=dyn_ear_03_l}
    componentAppliedLinearAccClamp=<10000 10000 10000>
    componentLinearAccScale=<1 1 1>
    componentLinearVelScale=<1 1 1>
    gravityOverride {z=-980}
    physicsBodyDefinitions [{
      boundBone {boneName=dyn_ear_03_l}
      boxExtents=<20 30 30>
      constraintSetup {angularLimitsMax=<0 0 30>, angularLimitsMin=<0 0 -7.5>, angularTarget=<1 0 0>}
      localJointOffset=<4 0 0>
     }]
    simSpaceSettings {maxAngularAcceleration=3, maxAngularVelocity=1, simSpaceAngularAlpha=1}
   }
   n33 = Anim Dynamics - Bone: dyn_ear_04_l!(ComponentPose=@n32, Alpha=1, GravityOverride:Vector="0,0,-980") -> (Pose) {
    angularDampingOverride=1
    angularSpringConstant=15000
    bAngularSpring=true
    bOverrideAngularDamping=true
    bUseGravityOverride=true
    gravityScale=0.35
    simulationSpace=RootRelative
    boundBone {boneName=dyn_ear_04_l}
    componentAppliedLinearAccClamp=<10000 10000 10000>
    componentLinearAccScale=<1 1 1>
    componentLinearVelScale=<1 1 1>
    gravityOverride {z=-980}
    physicsBodyDefinitions [{
      boundBone {boneName=dyn_ear_04_l}
      boxExtents=<15 25 25>
      constraintSetup {angularLimitsMax=<0 0 30>, angularLimitsMin=<0 0 -7.5>, angularTarget=<1 0 0>}
      localJointOffset=<4 0 0>
     }]
    simSpaceSettings {maxAngularAcceleration=3, maxAngularVelocity=1, simSpaceAngularAlpha=1}
   }
   n34 = Anim Dynamics - Bone: dyn_ear_01_r!(ComponentPose=@n33, Alpha=1, GravityOverride:Vector="0,0,-980") -> (Pose) {
    angularDampingOverride=1
    angularSpringConstant=15000
    bAngularSpring=true
    bOverrideAngularDamping=true
    bUseGravityOverride=true
    gravityScale=0.35
    simulationSpace=RootRelative
    boundBone {boneName=dyn_ear_01_r}
    componentAppliedLinearAccClamp=<10000 10000 10000>
    componentLinearAccScale=<1 1 1>
    componentLinearVelScale=<1 1 1>
    gravityOverride {z=-980}
    physicsBodyDefinitions [{
      boundBone {boneName=dyn_ear_01_r}
      boxExtents=<25 35 35>
      constraintSetup {angularLimitsMax=<0 0 10>, angularLimitsMin=<0 0 -15>, angularTarget=<1 0 0>}
      localJointOffset=<3.5 0 0>
     }]
    simSpaceSettings {maxAngularAcceleration=3, maxAngularVelocity=1, simSpaceAngularAlpha=1}
   }
   n35 = Anim Dynamics - Bone: dyn_ear_02_r!(ComponentPose=@n34, Alpha=1, GravityOverride:Vector="0,0,-980") -> (Pose) {
    angularDampingOverride=1
    angularSpringConstant=15000
    bAngularSpring=true
    bOverrideAngularDamping=true
    bUseGravityOverride=true
    gravityScale=0.35
    simulationSpace=RootRelative
    boundBone {boneName=dyn_ear_02_r}
    componentAppliedLinearAccClamp=<10000 10000 10000>
    componentLinearAccScale=<1 1 1>
    componentLinearVelScale=<1 1 1>
    gravityOverride {z=-980}
    physicsBodyDefinitions [{
      boundBone {boneName=dyn_ear_02_r}
      boxExtents=<25 35 35>
      constraintSetup {angularLimitsMax=<0 0 7.5>, angularLimitsMin=<0 0 -30>, angularTarget=<1 0 0>}
      localJointOffset=<4 0 0>
     }]
    simSpaceSettings {maxAngularAcceleration=3, maxAngularVelocity=1, simSpaceAngularAlpha=1}
   }
   n36 = Anim Dynamics - Bone: dyn_ear_03_r!(ComponentPose=@n35, Alpha=1, GravityOverride:Vector="0,0,-980") -> (Pose) {
    angularDampingOverride=1
    angularSpringConstant=15000
    bAngularSpring=true
    bOverrideAngularDamping=true
    bUseGravityOverride=true
    gravityScale=0.35
    simulationSpace=RootRelative
    boundBone {boneName=dyn_ear_03_r}
    componentAppliedLinearAccClamp=<10000 10000 10000>
    componentLinearAccScale=<1 1 1>
    componentLinearVelScale=<1 1 1>
    gravityOverride {z=-980}
    physicsBodyDefinitions [{
      boundBone {boneName=dyn_ear_03_r}
      boxExtents=<20 30 30>
      constraintSetup {angularLimitsMax=<0 0 7.5>, angularLimitsMin=<0 0 -30>, angularTarget=<1 0 0>}
      localJointOffset=<4 0 0>
     }]
    simSpaceSettings {maxAngularAcceleration=3, maxAngularVelocity=1, simSpaceAngularAlpha=1}
   }
   n37 = Anim Dynamics - Bone: dyn_ear_04_r!(ComponentPose=@n36, Alpha=1, GravityOverride:Vector="0,0,-980") -> (Pose) {
    angularDampingOverride=1
    angularSpringConstant=15000
    bAngularSpring=true
    bOverrideAngularDamping=true
    bUseGravityOverride=true
    gravityScale=0.35
    simulationSpace=RootRelative
    boundBone {boneName=dyn_ear_04_r}
    componentAppliedLinearAccClamp=<10000 10000 10000>
    componentLinearAccScale=<1 1 1>
    componentLinearVelScale=<1 1 1>
    gravityOverride {z=-980}
    physicsBodyDefinitions [{
      boundBone {boneName=dyn_ear_04_r}
      boxExtents=<15 25 25>
      constraintSetup {angularLimitsMax=<0 0 7.5>, angularLimitsMin=<0 0 -30>, angularTarget=<1 0 0>}
      localJointOffset=<4 0 0>
     }]
    simSpaceSettings {maxAngularAcceleration=3, maxAngularVelocity=1, simSpaceAngularAlpha=1}
   }
   Save cached pose 'DynamicsCache'(Pose=Blend Poses by bool(BlendPose_0=Blend Poses by bool(BlendPose_0=@n17, BlendPose_1=@n4), BlendPose_1=Component To Local(ComponentPose=@n37))) {cacheName=DynamicsCache}
  ]}
 EventGraph {nodes [
   Event BlueprintUpdateAnimation()
   > n1 = Sequence!(execute) -> (then_0, then_1, then_2:Exec)
   n2 = Branch(execute=@n1.then_2, Condition=Get bAnimDynamicsEnabled() -> (bAnimDynamicsEnabled:bool))
   > n3 = Branch(Condition=Get bIsInFrontEnd())
   > Set earOffsets(earOffsets:double=0.5) -> (Output_Get:double)
   Set earOffsets(execute=@n3.else, earOffsets=1)
   Set earOffsets(execute=@n2.else, earOffsets=0)
   n4 = "Gravity Override State Function (Parameter Asset) (Nativized)"(
    execute=@n1.then_0
    CharacterPartAnimInstance=Self-Reference()
    GameTimeInSeconds=GetGameTimeInSeconds()
    Parameters=&/F_MED_IndoorLaceScore_Pack_Parameters
   )
   > n5 = Set Gravity Override(Gravity Override=@n4)
   > Branch(Condition="Equal (Vector)"!(A:Vector=@n5, B:Vector, ErrorTolerance:float) -> (ReturnValue:bool=~) {is=title K2Node_PromotableOperator})
   > Set Gravity Override(Gravity Override="0,0,-400")
   n6 = Branch(execute=@n1.then_1, Condition=Get bIsInFrontEnd())
   > Set RB Sim Space(RB Sim Space:Sim Space Settings Structure=Get RB FE Space() -> (RB FE Space:Sim Space Settings Structure)) -> (Output_Get:Sim Space Settings Structure)
   Set RB Sim Space(execute=@n6.else, RB Sim Space=Get RB BR Space() -> (RB BR Space:Sim Space Settings Structure))
   n7 = Event OnCharacterCustomizationLoaded()
   n8 = Get Array Item with EFort Custom Part Type(Array:Array of Custom Character Part Object References=@n7.CustomParts) -> (Output:Custom Character Part Object Reference)
   > Set isOnMale(isOnMale:bool="Equal (Enum)"(
     A:EFortCustomGender Enum=Get GenderPermitted!(self:Custom Character Part Object Reference=@n8) -> (GenderPermitted:EFortCustomGender Enum=Invalid)
     B:EFortCustomGender Enum=Male
    )) -> (Output_Get:bool)
   TryGetPawnOwner()
  ]}
]
variables [
 earOffsets {type=number}
 Gravity Override {type=Vector}
 isOnMale {type=boolean}
 Owning Pawn {type=Pawn, default=null}
 RB BR Space {
  type=SimSpaceSettings
  default {
   dampingAlpha=0.25
   maxAngularAcceleration=1600
   maxAngularVelocity=4
   maxLinearAcceleration=1000
   maxLinearVelocity=300
   velocityScaleZ=0.5
   worldAlpha=1
   externalLinearDragV=<1.2 1.2 1.2>
  }
 }
 RB FE Space {
  type=SimSpaceSettings
  default {
   dampingAlpha=1
   maxAngularAcceleration=1
   maxAngularVelocity=5
   maxLinearAcceleration=800
   maxLinearVelocity=400
   velocityScaleZ=1
   worldAlpha=1
   externalLinearDragV=<1.2 1.2 1.2>
  }
 }
 RB Sim Space {
  type=SimSpaceSettings
  default {
   dampingAlpha=0.25
   maxAngularAcceleration=1600
   maxAngularVelocity=4
   maxLinearAcceleration=1000
   maxLinearVelocity=300
   velocityScaleZ=0.5
   worldAlpha=1
   externalLinearDragV=<1.2 1.2 1.2>
  }
 }
]
